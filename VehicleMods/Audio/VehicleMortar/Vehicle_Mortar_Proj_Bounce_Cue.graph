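# Sound cue for the bounce of a vehicle mortar projectile.
# A Random node picks one of six equally weighted bounce waves, feeding a Modulator that holds pitch fixed.
# It plays at reduced volume in the ranged-weapon projectile effects sound class, using the shared grenade bounce attenuation.
# A concurrency override caps simultaneous bounces at three and blocks new ones beyond that.
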SoundCue /VehicleMods/Audio/VehicleMortar/Vehicle_Mortar_Proj_Bounce_Cue
root=Modulator_5
nodes [
 {id=Modulator_5, children [Random_0], properties {pitchMax=1, pitchMin=1}}
 {id=Random_0, children [WavePlayer_0,WavePlayer_1,WavePlayer_2,WavePlayer_3,WavePlayer_4,WavePlayer_5], properties {weights [1,1,1,1,1,1]}}
 {id=WavePlayer_0, properties {soundWaveAssetPtr=&/Vehicle_Mortar_Proj_Bounce_01}}
 {id=WavePlayer_1, properties {soundWaveAssetPtr=&/Vehicle_Mortar_Proj_Bounce_02}}
 {id=WavePlayer_2, properties {soundWaveAssetPtr=&/Vehicle_Mortar_Proj_Bounce_03}}
 {id=WavePlayer_3, properties {soundWaveAssetPtr=&/Vehicle_Mortar_Proj_Bounce_04}}
 {id=WavePlayer_4, properties {soundWaveAssetPtr=&/Vehicle_Mortar_Proj_Bounce_05}}
 {id=WavePlayer_5, properties {soundWaveAssetPtr=&/Vehicle_Mortar_Proj_Bounce_06}}
]
properties {
 attenuationSettings=/Game/Sounds/Attenuation/Fort_Grenade_Bounce_Att
 bOverrideConcurrency=true
 maxDistance=5500
 soundClassObject=&Fort_WeaponRanged_Projectile_Effects
 volumeMultiplier=0.3
 concurrencyOverrides {maxCount=3, resolutionRule=PreventNew}
}
