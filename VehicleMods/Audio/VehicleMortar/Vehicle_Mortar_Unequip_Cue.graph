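# Sound cue for unequipping the vehicle mortar mod, played as exterior vehicle audio.
# A Random node picks one of three equally weighted unequip waves, and a Modulator above it adds slight pitch and volume variation to each play.
# It uses the weapon reload attenuation and the exterior vehicles sound class, with a reduced overall volume.
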SoundCue /VehicleMods/Audio/VehicleMortar/Vehicle_Mortar_Unequip_Cue
root=Modulator_0
nodes [
 {id=Modulator_0, children [Random_0], properties {pitchMax=1.01, pitchMin=0.98, volumeMax=1.01, volumeMin=0.99}}
 {id=Random_0, children [WavePlayer_0,WavePlayer_1,WavePlayer_2], properties {weights [1,1,1]}}
 {id=WavePlayer_0, properties {soundWaveAssetPtr=&/Vehicle_Mortar_Unequip_01}}
 {id=WavePlayer_1, properties {soundWaveAssetPtr=&/Vehicle_Mortar_Unequip_02}}
 {id=WavePlayer_2, properties {soundWaveAssetPtr=&/Vehicle_Mortar_Unequip_03}}
]
properties {
 attenuationSettings=&Fort_Weapon_Reload_Att
 maxDistance=2250
 soundClassObject=&Fort_Vehicles_Exterior
 volumeMultiplier=0.6
}
